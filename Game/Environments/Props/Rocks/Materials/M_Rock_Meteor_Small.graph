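# Lit surface material for small meteor rock props with an animated emissive glow, also enabled for skeletal meshes and mesh particles.
# A static switch picks crack-texture glow or procedural glow from an object-centered sphere mask, specular mask and inverted diffuse, both modulated by panning noise and HDR-remapped.
# Ray-traced emissive is scaled by time of day for Lumen, and all attributes pass through ThinFilmFunction.
# Emissive_1 and BlackBody branches are unconnected.
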Material /Game/Environments/Props/Rocks/Materials/M_Rock_Meteor_Small
graph {
 properties {
  bUseMaterialAttributes=true
  bUsedWithMeshDeformer=true
  bUsedWithMeshParticles=true
  bUsedWithSkeletalMesh=true
  pixelDepthOffsetMode=PDOM_Legacy
  usedShadingModels=Default Lit
 }
 nodes [
  n1 = ScalarParameter() {defaultValue=1, parameterName=InternalGlowMult}
  n2 = VectorParameter() {parameterName=GlowColor, defaultValue=<rgba 2 0.2173 0.0938 0>}
  n3 = WorldPosition()
  n4 = Panner(Coordinate=Divide(A=ComponentMask(None=@n3.XYZ) {g=true, r=true}, B=ScalarParameter() {defaultValue=151.7832, parameterName=NoiseTiling})) {speedX=0.04, speedY=0.1}
  n5 = Multiply(A=@n2.RGB, B=LinearInterpolate(Alpha=TextureSample(UVs=@n4) {texture=&T_NoisePack_01}.G) {constA=0.4})
  n6 = TextureSampleParameter2D() {parameterName=Emissive_2, samplerType=SAMPLERTYPE_Masks, texture=/Game/Environments/Props/Rocks/Textures/T_Meteor_Cracked_E2}
  n7 = VectorParameter() {parameterName=GlowTint, defaultValue=<rgba 14.0704 15 1.0938 1>}
  n8 = Multiply(
   A=@n1
   B=MaterialFunctionCall(LDR Input=Multiply(A=@n5, B=@n6.R), HDR Tint=@n7.RGB) {materialFunction=&DeriveHDRfromLDR}.RemappedRange
  )
  Add(
   A=@n8
   B=Multiply(
    A=TextureSampleParameter2D() {parameterName=Emissive_1, texture=/Game/Environments/Props/Rocks/Textures/T_Meteor_Cracked_E}.RGB
    B=ScalarParameter() {defaultValue=1, parameterName=EmissiveBrightnes}
   )
  )
  n9 = TextureSampleParameter2D() {parameterName=Diffuse, texture=/Game/Environments/Props/Rocks/Textures/T_Meteor_low_D}
  n10 = TextureSampleParameter2D() {parameterName=SpecularMask, samplerType=SAMPLERTYPE_Masks, texture=/Game/Environments/Props/Rocks/Textures/T_Meteor_low_S}
  n11 = ScalarParameter() {defaultValue=0.1, parameterName=Roughness_Min}
  n12 = ScalarParameter() {defaultValue=1, parameterName=Roughness_Max}
  n13 = OneMinus(None=@n9.RGB)
  n14 = ScalarParameter() {defaultValue=4, parameterName=Exp_Spec}
  n15 = ObjectRadius()
  n16 = ScalarParameter() {defaultValue=1, parameterName=GlowSphereMaskSizeMult}
  n17 = Power(
   Base=SphereMask(A=WorldPosition().XYZ, B=ObjectPositionWS(), Radius=Multiply(A=@n15, B=@n16)) {hardnessPercent=0}
   Exp=ScalarParameter() {defaultValue=3.4805, parameterName=GlowSpherePower}
  )
  n18 = Panner(Coordinate=TextureCoordinate()) {speedX=0.04, speedY=0.1}
  n19 = Multiply(
   A=Multiply(A=Power(Base=@n10.B, Exp=@n14), B=Multiply(A=@n17) {constB=2})
   B=LinearInterpolate(Alpha=TextureSample(UVs=@n18) {texture=&T_NoisePack_01}.G) {constA=0.4}
  )
  n20 = Clamp(None=Multiply(A=Multiply(A=@n13, B=@n19) {constA=1}, B=@n2.RGB))
  n21 = Multiply(A=@n13, B=@n7.RGB)
  n22 = StaticSwitchParameter(
   True=@n8
   False=Multiply(
    A=@n1
    B=MaterialFunctionCall(LDR Input=@n20, HDR Tint=@n21) {materialFunction=&DeriveHDRfromLDR}.RemappedRange
   )
  ) {parameterName=UseEmissive_Texture}
  n23 = ScalarParameter() {defaultValue=1, parameterName=LumenEmissiveMultMorning, sortPriority=34}
  n24 = ScalarParameter() {defaultValue=1, parameterName=LumenEmissiveMultDay, sortPriority=35}
  n25 = ScalarParameter() {defaultValue=1, parameterName=LumenEmissiveMultEvening, sortPriority=36}
  n26 = ScalarParameter() {defaultValue=1, parameterName=LumenEmissiveMultNight, sortPriority=37}
  n27 = RayTracingQualitySwitch(
   Normal=@n22
   RayTraced=Multiply(
    A=MaterialFunctionCall(Value_Morning=@n23, Value_Day=@n24, Value_Evening=@n25, Value_Night=@n26) {materialFunction=&Scalar_Time_of_Day}.Result
    B=@n22
   )
  )
  n28 = Constant()
  MaterialFunctionCall!(
   Alpha
   Material Attributes=MakeMaterialAttributes!(
    BaseColor=@n9.RGB
    Metallic=@n10.G
    Specular=@n10.G
    Roughness=LinearInterpolate(A=@n11, B=@n12, Alpha=@n10.B)
    Anisotropy
    EmissiveColor=StaticSwitchParameter(True=@n27, False=@n28) {defaultValue=true, parameterName=UseEmissive}
    Opacity
    OpacityMask
    Normal=TextureSampleParameter2D() {parameterName=Normal, samplerType=SAMPLERTYPE_Normal, texture=/Game/Environments/Props/Rocks/Textures/T_Meteor_low_N}.RGB
    Tangent
    WorldPositionOffset
    SubsurfaceColor
    ClearCoat
    ClearCoatRoughness
    AmbientOcclusion
    Refraction
    CustomizedUVs_0
    CustomizedUVs_1
    CustomizedUVs_2
    CustomizedUVs_3
    CustomizedUVs_4
    CustomizedUVs_5
    CustomizedUVs_6
    CustomizedUVs_7
    PixelDepthOffset
    ShadingModel
    Displacement
   )
  ) {materialFunction=/Game/Weapons/WeaponSkins/MaterialFunction/ThinFilmFunction}
  Multiply(A=BlackBody!(Temp=@n6.R), B=ScalarParameter() {parameterName=Glow Brightness})
 ]
}
