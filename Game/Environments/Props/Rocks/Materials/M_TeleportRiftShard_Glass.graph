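# Translucent, refractive glass material for the Teleport Rift Shard rock prop, usable on static, skeletal and Niagara meshes.
# One Fresnel term blends transparency, refraction and a green-to-cyan tint.
# Emissive fakes reflection and lighting with a tinted cubemap, a backlit sun pass-through glint and soft plus sharp sun highlights, via FX emissive adjustments.
# Actor and tiled noise normals combine for surface detail, and BlackHoleFunction takes a world-position parameter.
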Material /Game/Environments/Props/Rocks/Materials/M_TeleportRiftShard_Glass
graph {
 properties {
  bEnableResponsiveAA=true
  bScreenSpaceReflections=true
  bUsedWithMeshDeformer=true
  bUsedWithNiagaraMeshParticles=true
  bUsedWithSkeletalMesh=true
  blendMode=BLEND_Translucent
  physMaterial=/Game/FortressPhysicalMaterials/Glass
  pixelDepthOffsetMode=PDOM_Legacy
  previewMesh=/BRCosmetics/Environments/Props/Rocks/Meshes/SM_TeleportRiftShard
  refractionMethod=RM_IndexOfRefraction
  translucencyDirectionalLightingIntensity=5
  translucencyLightingMode=TLM_VolumetricPerVertexDirectional
  usedShadingModels=Default Lit
 }
 nodes [
  n1 = Fresnel(
   ExponentIn=ScalarParameter() {defaultValue=3, group=Fresnel, parameterName=Base Fresnesl Exponent}
   BaseReflectFractionIn=ScalarParameter() {defaultValue=0.04, group=Fresnel, parameterName=Base Reflect Fraction}
  )
  LinearInterpolate(
   A=ScalarParameter() {defaultValue=0.25, group=Fresnel, parameterName=Fresnel Inner Transparency}
   B=ScalarParameter() {defaultValue=1, group=Fresnel, parameterName=Fresnel Outer Transparency}
   Alpha=@n1
  )
  LinearInterpolate(
   A=ScalarParameter() {defaultValue=1.1, group=Refraction, parameterName=Refraction Center, sliderMax=2}
   B=ScalarParameter() {defaultValue=0.25, group=Refraction, parameterName=Refraction Fresnel, sliderMax=2}
   Alpha=@n1
  )
  n2 = LinearInterpolate(
   A=VectorParameter() {group=Color, parameterName=Base Color, defaultValue=<rgba 0.0345 0.2552 0.0013 0>}.RGB
   B=VectorParameter() {group=Color, parameterName=Base Color Fresnel, defaultValue=<rgba 0.138 0.4131 0.5521 0>}.RGB
   Alpha=@n1
  )
  n3 = ReflectionVectorWS() {bNormalizeCustomWorldNormal=true}
  n4 = MaterialFunctionCall() {materialFunction=/Engine/Functions/GetPostProcessFunctions/GetAmbientCubemapTint}
  n5 = AtmosphericLightVector()
  n6 = DotProduct(A=Multiply(A=ComponentMask(None=@n5) {b=true, g=true, r=true}) {constB=-1}, B=CameraVectorWS())
  n7 = ScalarParameter() {defaultValue=155, group=Spec, parameterName=Pass Through Spec Power}
  n8 = Add(
   A=Multiply(
    A=TextureSampleParameterCube!(UVs=@n3, Apply View MipBias) {parameterName=MaterialExpressionTextureSampleParameterCube_0, texture=/Game/Packages/Fortress_Sky/TexturesHDR/T_hdrsky_test03_low}.R
    B=ComponentMask(None=@n4.AmbientCubemapTint) {b=true, g=true, r=true}
   )
   B=Multiply(A=Power(Base=@n6, Exp=@n7) {constExponent=20}, B=ScalarParameter() {defaultValue=12, group=Spec, parameterName=Pass Through Spec Intense})
  )
  n9 = AtmosphericLightVector()
  n10 = Saturate(None=DotProduct(A=ComponentMask(None=@n9) {b=true, g=true, r=true}, B=ReflectionVectorWS()))
  n11 = Multiply(
   A=Power(Base=@n10, Exp=ScalarParameter() {defaultValue=45, group=Spec, parameterName=SoftSpecPower}) {constExponent=20}
   B=ScalarParameter() {defaultValue=0.3, group=Spec, parameterName=SoftSpecLevel}
  )
  n12 = Multiply(
   A=Power(Base=@n10, Exp=ScalarParameter() {defaultValue=1000, group=Spec, parameterName=Spec Highlight Power}) {constExponent=20}
   B=ScalarParameter() {defaultValue=3, group=Spec, parameterName=Spec Intense}
  )
  MaterialFunctionCall(Emissive=Add(A=Multiply(A=@n2, B=@n8), B=Add(A=@n11, B=@n12))) {materialFunction=&MF_FX_EmissiveAdjustments}
  n13 = VectorParameter() {group=FX, parameterName=WorldPostionOfBlackHole, defaultValue=<rgba -30000 25536 1952 0>}
  MaterialFunctionCall!(WorldPostionOfBlackHole=AppendVector(A=@n13.RGB, B=@n13.RGB), World Position) {materialFunction=/Game/Packages/Fortnite_UniqueMaterials_1/BlackHoleFunction}
  n14 = TextureSampleParameter2D() {group=Normal, parameterName="Actor Normal ", samplerType=SAMPLERTYPE_Normal, texture=/Game/Global/Textures/Default/Noise/T_Noise_Normal_01}
  n15 = VectorParameter() {group=Normal, parameterName=ActorNormal Intensity, defaultValue=<rgba 0.1 0.1 1 0>}
  n16 = TextureSampleParameter2D(UVs=Multiply(A=TextureCoordinate(), B=ScalarParameter() {defaultValue=1, group=Normal, parameterName=TileNormal Scale})) {group=Normal, parameterName=Tile Normal, samplerType=SAMPLERTYPE_Normal, texture=&T_Noise_Normal_01}
  n17 = Multiply(A=@n16.RGB, B=VectorParameter() {group=Normal, parameterName=Normal Intensity, defaultValue=<rgba 0.1 0.1 1 0>}.RGB)
  n18 = Constant3Vector() {constant=<rgba 1 1 0 0>}
  Normalize(VectorInput=Add(A=Multiply(A=@n14.RGB, B=@n15.RGB), B=Multiply(A=@n17, B=@n18)))
 ]
}
